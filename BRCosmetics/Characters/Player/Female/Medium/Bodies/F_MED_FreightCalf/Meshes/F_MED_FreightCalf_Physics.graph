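# Physics asset for the FreightCalf female medium character body, setting up a dynamic jacket with body colliders.
# Five four-segment capsule chains (left, right, center, two sides) hang with angular limits from zero-size BASE bodies constrained to spine_02.
# Cross-constraints, mostly angular-free, tie some neighboring chains together; the right BASE is also angular-free, unlike the other locked anchors.
# Capsules on pelvis, spine, thighs, glutes, right arm and a clip bone complete the setup.
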PhysicsAsset /BRCosmetics/Characters/Player/Female/Medium/Bodies/F_MED_FreightCalf/Meshes/F_MED_FreightCalf_Physics
bodies {
 deform_glute_base_l [{
   shapeType=Capsule
   length=12
   radius=7.5
   shapeName=deform_glute_base_l_capsule
   center=<4.7435 1.9988 1.7108>
   rotation=<rot -65.3813 -84.9285 5.3029>
  }]
 deform_glute_base_r [{
   shapeType=Capsule
   length=7.5
   radius=9
   shapeName=deform_glute_base_l_capsule
   center=<-6.1923 -5.1083 -1.4486>
   rotation=<rot -85.2188 -91.0477 156.1399>
  }]
 dyn_clip_01 [{
   shapeType=Capsule
   length=4.9128
   radius=4.737
   shapeName=dyn_clip_01_capsule
   center=<0.0717 -2.4438 -2.8333>
   rotation=<rot -11.3944 -94.3712 -30.2207>
  }]
 dyn_jacketSide_01_l [{shapeType=Capsule, length=14, radius=3, shapeName=dyn_jacketSide_01_l_capsule, center=<3 3 4>, rotation=<rot -0.0004 0.0001 -19>}]
 dyn_jacketSide_01_l_BASE [{shapeType=Capsule, length=0, radius=0, shapeName=dyn_jacketSide_01_l_BASE_capsule}]
 dyn_jacketSide_01_r [{shapeType=Capsule, length=19.5, radius=2.5, shapeName=dyn_jacketSide_01_r_capsule, center=<-3 -3 -2>, rotation=<rot -0.0007 0.0001 -15>}]
 dyn_jacketSide_01_r_BASE [{shapeType=Capsule, length=0, radius=0, shapeName=dyn_jacketSide_01_r_BASE_capsule, rotation=<rot -0.0004 0.0001 0>}]
 dyn_jacketSide_02_l [{shapeType=Capsule, length=14, radius=3, shapeName=dyn_jacketSide_02_l_capsule, center=<3 3 4>, rotation=<rot -0.0004 0.0001 -19>}]
 dyn_jacketSide_02_r [{shapeType=Capsule, length=19.5, radius=2.5, shapeName=dyn_jacketSide_02_r_capsule, center=<-3 -3 -2>, rotation=<rot -0.0007 0.0001 -15>}]
 dyn_jacketSide_03_l [{shapeType=Capsule, length=14, radius=3, shapeName=dyn_jacketSide_03_l_capsule, center=<3 3 4>, rotation=<rot -0.0004 0.0001 -19>}]
 dyn_jacketSide_03_r [{shapeType=Capsule, length=19.5, radius=2.9085, shapeName=dyn_jacketSide_03_r_capsule, center=<-3 -3 -2>, rotation=<rot -0.0011 0.0001 -15>}]
 dyn_jacketSide_04_l [{shapeType=Capsule, length=14.5, radius=3, shapeName=dyn_jacketSide_04_l_capsule, center=<3 3 4>, rotation=<rot -0.0004 0.0001 -19>}]
 dyn_jacketSide_04_r [{shapeType=Capsule, length=19.5, radius=3.152, shapeName=dyn_jacketSide_04_r_capsule, center=<-3 -3 -2>, rotation=<rot 0.0006 0.0001 165>}]
 dyn_jacket_01_c [{shapeType=Capsule, length=24.5, radius=2, shapeName=dyn_jacket_01_c_capsule, center=<3 0.0161 2.5>, rotation=<rot -0.0002 0 -95>}]
 dyn_jacket_01_c_BASE [{shapeType=Capsule, length=0, radius=0, shapeName=dyn_jacket_01_c_BASE_capsule, center=<3 0 0>, rotation=<rot -0.0002 0 -95>}]
 dyn_jacket_01_l [{shapeType=Capsule, length=8.5, radius=3, shapeName=dyn_jacket_01_l_capsule, center=<3 3 1>, rotation=<rot -0.0003 0.0003 5>}]
 dyn_jacket_01_l_BASE [{shapeType=Capsule, length=0, radius=0, shapeName=dyn_jacket_01_l_BASE_capsule}]
 dyn_jacket_01_r [{shapeType=Capsule, length=11, radius=2.5, shapeName=dyn_jacket_01_l_capsule, center=<-3 -3 0>, rotation=<rot -0.0001 0.0002 0>}]
 dyn_jacket_01_r_BASE [{shapeType=Capsule, length=0, radius=0, shapeName=dyn_jacket_01_l_BASE_capsule, rotation=<rot -0.0001 0.0002 180>}]
 dyn_jacket_02_c [{shapeType=Capsule, length=24.5, radius=3, shapeName=dyn_jacket_02_c_capsule, center=<3 0.0161 2.5>, rotation=<rot -0.0002 0 -95>}]
 dyn_jacket_02_l [{shapeType=Capsule, length=8, radius=3, shapeName=dyn_jacket_02_l_capsule, center=<3 3 1>, rotation=<rot -0.0003 0.0003 5>}]
 dyn_jacket_02_r [{shapeType=Capsule, length=11, radius=2.5, shapeName=dyn_jacket_02_l_capsule, center=<-3 -3 0>, rotation=<rot -0.0001 0.0002 0>}]
 dyn_jacket_03_c [{shapeType=Capsule, length=24.5, radius=3, shapeName=dyn_jacket_03_c_capsule, center=<3 0.0161 2.5>, rotation=<rot -0.0002 0 -95>}]
 dyn_jacket_03_l [{shapeType=Capsule, length=8, radius=3, shapeName=dyn_jacket_03_l_capsule, center=<3 3 1>, rotation=<rot -0.0003 0.0003 5>}]
 dyn_jacket_03_r [{shapeType=Capsule, length=11, radius=2.5, shapeName=dyn_jacket_03_l_capsule, center=<-3 -3 0>, rotation=<rot -0.0001 0.0002 0>}]
 dyn_jacket_04_c [{shapeType=Capsule, length=24.5, radius=3, shapeName=dyn_jacket_04_c_capsule, center=<3 0.0161 2.5>, rotation=<rot -0.0002 0 -95>}]
 dyn_jacket_04_l [{shapeType=Capsule, length=8, radius=3.5, shapeName=dyn_jacket_04_l_capsule, center=<3 3 1>, rotation=<rot -0.0003 0.0003 5>}]
 dyn_jacket_04_r [{shapeType=Capsule, length=11, radius=2.5, shapeName=dyn_jacket_04_l_capsule, center=<-3 -3 0>, rotation=<rot -0.0002 0.0003 5>}]
 lowerarm_r [{shapeType=Capsule, length=21.5, radius=2, shapeName=lowerarm_r_capsule, center=<-10.8355 -0.0001 0>, rotation=<rot 90 -90 -90.0004>}]
 pelvis [{shapeType=Capsule, length=14.4708, radius=8.2736, shapeName=pelvis_capsule}]
 spine_01 [{shapeType=Capsule, length=9.4126, radius=11.2106, shapeName=spine_01_capsule, rotation=<rot -90 -177.767 177.7669>}]
 spine_02 [{shapeType=Capsule, length=16.2627, radius=13.871, shapeName=spine_02_capsule, center=<-1.6909 0 0>, rotation=<rot -90 179.9208 -179.9209>}]
 thigh_l [{shapeType=Capsule, length=28, radius=9, shapeName=thigh_l_capsule, center=<-15.7844 -0.0725 1.1035>, rotation=<rot -90 -90.5672 85.5669>}]
 thigh_r [{shapeType=Capsule, length=20.5, radius=8, shapeName=thigh_l_capsule, center=<15.5661 -1.1653 -0.017>, rotation=<rot -90 -90 -100.0003>}]
 upperarm_r [{shapeType=Capsule, length=21.5, radius=2, shapeName=upperarm_r_capsule, center=<-10.8355 -0.0001 0>, rotation=<rot 90 -90 -90.0005>}]
}
constraints [
 {bone1Name=dyn_jacket_01_l, bone2Name=dyn_jacket_01_l_BASE, swing1LimitDegrees=40, swing2LimitDegrees=10, twistLimitDegrees=5}
 {bone1Name=dyn_jacket_02_l, bone2Name=dyn_jacket_01_l, swing1LimitDegrees=40, swing2LimitDegrees=10, twistLimitDegrees=5}
 {bone1Name=dyn_jacket_03_l, bone2Name=dyn_jacket_02_l, swing1LimitDegrees=40, swing2LimitDegrees=10, twistLimitDegrees=5}
 {bone1Name=dyn_jacket_04_l, bone2Name=dyn_jacket_03_l, swing1LimitDegrees=40, swing2LimitDegrees=10, twistLimitDegrees=5}
 {
  bone1Name=dyn_jacket_01_l_BASE
  bone2Name=spine_02
  swing1LimitDegrees=45
  swing1Motion=Locked
  swing2LimitDegrees=45
  swing2Motion=Locked
  twistLimitDegrees=45
  twistMotion=Locked
 }
 {
  bone1Name=thigh_l
  bone2Name=pelvis
  swing1LimitDegrees=45
  swing1Motion=Locked
  swing2LimitDegrees=45
  swing2Motion=Locked
  twistLimitDegrees=45
  twistMotion=Locked
 }
 {
  bone1Name=dyn_jacketSide_01_l_BASE
  bone2Name=spine_02
  swing1LimitDegrees=45
  swing1Motion=Locked
  swing2LimitDegrees=45
  swing2Motion=Locked
  twistLimitDegrees=45
  twistMotion=Locked
 }
 {bone1Name=dyn_jacketSide_01_l, bone2Name=dyn_jacketSide_01_l_BASE, swing1LimitDegrees=40, swing2LimitDegrees=20, twistLimitDegrees=5}
 {bone1Name=dyn_jacketSide_02_l, bone2Name=dyn_jacketSide_01_l, swing1LimitDegrees=40, swing2LimitDegrees=20, twistLimitDegrees=5}
 {bone1Name=dyn_jacketSide_03_l, bone2Name=dyn_jacketSide_02_l, swing1LimitDegrees=40, swing2LimitDegrees=20, twistLimitDegrees=5}
 {bone1Name=dyn_jacketSide_04_l, bone2Name=dyn_jacketSide_03_l, swing1LimitDegrees=40, swing2LimitDegrees=20, twistLimitDegrees=5}
 {
  bone1Name=dyn_jacket_01_c_BASE
  bone2Name=spine_02
  swing1LimitDegrees=45
  swing1Motion=Locked
  swing2LimitDegrees=45
  swing2Motion=Locked
  twistLimitDegrees=45
  twistMotion=Locked
 }
 {
  bone1Name=dyn_jacket_01_c
  bone2Name=dyn_jacket_01_c_BASE
  swing1LimitDegrees=10
  swing1Motion=Locked
  swing2LimitDegrees=40
  twistLimitDegrees=5
  twistMotion=Locked
 }
 {
  bone1Name=dyn_jacket_02_c
  bone2Name=dyn_jacket_01_c
  swing1LimitDegrees=10
  swing1Motion=Locked
  swing2LimitDegrees=40
  twistLimitDegrees=5
  twistMotion=Locked
 }
 {
  bone1Name=dyn_jacket_03_c
  bone2Name=dyn_jacket_02_c
  swing1LimitDegrees=10
  swing1Motion=Locked
  swing2LimitDegrees=40
  twistLimitDegrees=5
  twistMotion=Locked
 }
 {
  bone1Name=dyn_jacket_04_c
  bone2Name=dyn_jacket_03_c
  swing1LimitDegrees=10
  swing1Motion=Locked
  swing2LimitDegrees=40
  twistLimitDegrees=5
  twistMotion=Locked
 }
 {
  bone1Name=spine_01
  bone2Name=pelvis
  swing1LimitDegrees=45
  swing1Motion=Free
  swing2LimitDegrees=45
  swing2Motion=Free
  twistLimitDegrees=45
  twistMotion=Free
 }
 {
  bone1Name=deform_glute_base_l
  bone2Name=pelvis
  swing1LimitDegrees=45
  swing1Motion=Free
  swing2LimitDegrees=45
  swing2Motion=Free
  twistLimitDegrees=45
  twistMotion=Free
 }
 {
  bone1Name=dyn_jacket_01_r_BASE
  bone2Name=spine_02
  swing1LimitDegrees=45
  swing1Motion=Free
  swing2LimitDegrees=45
  swing2Motion=Free
  twistLimitDegrees=45
  twistMotion=Free
 }
 {
  bone1Name=dyn_jacket_01_r
  bone2Name=dyn_jacket_01_r_BASE
  swing1LimitDegrees=40
  swing2LimitDegrees=10
  twistLimitDegrees=10
  twistMotion=Locked
 }
 {bone1Name=dyn_jacket_02_r, bone2Name=dyn_jacket_01_r, swing1LimitDegrees=40, swing2LimitDegrees=10, twistLimitDegrees=10, twistMotion=Locked}
 {bone1Name=dyn_jacket_03_r, bone2Name=dyn_jacket_02_r, swing1LimitDegrees=40, swing2LimitDegrees=10, twistLimitDegrees=10, twistMotion=Locked}
 {bone1Name=dyn_jacket_04_r, bone2Name=dyn_jacket_03_r, swing1LimitDegrees=40, swing2LimitDegrees=10, twistLimitDegrees=10, twistMotion=Locked}
 {
  bone1Name=spine_02
  bone2Name=spine_01
  swing1LimitDegrees=45
  swing1Motion=Free
  swing2LimitDegrees=45
  swing2Motion=Free
  twistLimitDegrees=45
  twistMotion=Free
 }
 {
  bone1Name=thigh_r
  bone2Name=pelvis
  swing1LimitDegrees=45
  swing1Motion=Locked
  swing2LimitDegrees=45
  swing2Motion=Locked
  twistLimitDegrees=45
  twistMotion=Locked
 }
 {
  bone1Name=dyn_jacketSide_01_r_BASE
  bone2Name=spine_02
  swing1LimitDegrees=45
  swing1Motion=Locked
  swing2LimitDegrees=45
  swing2Motion=Locked
  twistLimitDegrees=45
  twistMotion=Locked
 }
 {bone1Name=dyn_jacketSide_01_r, bone2Name=dyn_jacketSide_01_r_BASE, swing1LimitDegrees=40, swing2LimitDegrees=5, twistLimitDegrees=10}
 {bone1Name=dyn_jacketSide_02_r, bone2Name=dyn_jacketSide_01_r, swing1LimitDegrees=40, swing2LimitDegrees=5, twistLimitDegrees=10}
 {bone1Name=dyn_jacketSide_03_r, bone2Name=dyn_jacketSide_02_r, swing1LimitDegrees=40, swing2LimitDegrees=5, twistLimitDegrees=10}
 {bone1Name=dyn_jacketSide_04_r, bone2Name=dyn_jacketSide_03_r, swing1LimitDegrees=40, swing2LimitDegrees=5, twistLimitDegrees=10}
 {
  bone1Name=dyn_clip_01
  bone2Name=pelvis
  swing1LimitDegrees=45
  swing1Motion=Free
  swing2LimitDegrees=45
  swing2Motion=Free
  twistLimitDegrees=45
  twistMotion=Free
 }
 {
  bone1Name=deform_glute_base_r
  bone2Name=pelvis
  swing1LimitDegrees=45
  swing1Motion=Free
  swing2LimitDegrees=45
  swing2Motion=Free
  twistLimitDegrees=45
  twistMotion=Free
 }
 {bone1Name=dyn_jacketSide_03_r, bone2Name=dyn_jacket_03_c, swing1LimitDegrees=45, swing2LimitDegrees=45, twistLimitDegrees=45}
 {
  bone1Name=dyn_jacket_02_l
  bone2Name=dyn_jacketSide_02_l
  swing1LimitDegrees=45
  swing1Motion=Free
  swing2LimitDegrees=45
  swing2Motion=Free
  twistLimitDegrees=45
  twistMotion=Free
 }
 {
  bone1Name=dyn_jacket_04_l
  bone2Name=dyn_jacketSide_04_l
  swing1LimitDegrees=45
  swing1Motion=Free
  swing2LimitDegrees=45
  swing2Motion=Free
  twistLimitDegrees=45
  twistMotion=Free
 }
 {
  bone1Name=dyn_jacket_04_c
  bone2Name=dyn_jacketSide_04_r
  swing1LimitDegrees=45
  swing1Motion=Free
  swing2LimitDegrees=45
  swing2Motion=Free
  twistLimitDegrees=45
  twistMotion=Free
 }
 {
  bone1Name=dyn_jacketSide_04_r
  bone2Name=dyn_jacket_04_r
  swing1LimitDegrees=45
  swing1Motion=Free
  swing2LimitDegrees=45
  swing2Motion=Free
  twistLimitDegrees=45
  twistMotion=Free
 }
 {
  bone1Name=upperarm_r
  bone2Name=spine_02
  swing1LimitDegrees=45
  swing1Motion=Free
  swing2LimitDegrees=45
  swing2Motion=Free
  twistLimitDegrees=45
  twistMotion=Free
 }
 {
  bone1Name=lowerarm_r
  bone2Name=upperarm_r
  swing1LimitDegrees=45
  swing1Motion=Free
  swing2LimitDegrees=45
  swing2Motion=Free
  twistLimitDegrees=45
  twistMotion=Free
 }
]
