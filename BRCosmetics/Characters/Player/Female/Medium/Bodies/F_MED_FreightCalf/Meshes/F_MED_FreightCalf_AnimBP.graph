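# Post-process style Anim Blueprint for the FreightCalf female body cosmetic, adding secondary motion to jacket, collar, clips and dangling chains.
# It copies the attached parent's pose, runs a Control Rig, then constrains and offsets chain bones.
# Anim Dynamics swings two clips, a spine-relative RigidBody simulates physics, and weighted constraints steady button and collars.
# Each update computes skydive collar values and state-based gravity and RigidBody settings through a deprecated function.
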AnimBlueprint /BRCosmetics/Characters/Player/Female/Medium/Bodies/F_MED_FreightCalf/Meshes/F_MED_FreightCalf_AnimBP
target_skeleton=&/F_MED_FreightCalf_Skeleton
graphs [
 AnimGraph {nodes [
   "Transform (Modify) Bone - Bone: dyn_chain_04_BASE"(Alpha=1, Translation="0,0,0", Rotation="-35.000,-75.000,0.000", Scale="1,1,1") {
    rotationMode=BMM_Additive
    translationMode=BMM_Additive
    alphaBoolBlend {blendOption=Linear}
    boneToModify {boneName=dyn_chain_04_BASE}
    rotation=<rot -35 -75 0>
   }
   "Transform (Modify) Bone"(Alpha=1, Translation="0,0,0", Rotation="0,0,0", Scale="1,1,1") {alphaBoolBlend {blendOption=Linear}, boneToModify=~}
   RigidBody!(ComponentPose, Alpha, OverrideWorldGravity:Vector="0,0,-980") -> (Pose) {
    bClampLinearTranslationLimitToRefPose=true
    bOverrideWorldGravity=true
    overridePhysicsAsset=&/F_MED_FreightCalf_Physics
    simulationSpace=BaseBoneSpace
    alphaBoolBlend {blendOption=Linear}
    baseBoneRef {boneName=spine_05}
    simSpaceSettings {
     maxAngularAcceleration=10
     maxAngularVelocity=10
     maxLinearAcceleration=500
     maxLinearVelocity=500
     worldAlpha=1
     externalLinearDragV=<0.2 0.2>
    }
   }
   RigidBody!(
    ComponentPose
    Alpha
    OverrideWorldGravity="0,0,-980"
    ComponentLinearAccScale:Vector="0.3,0.2,0.2"
    ComponentLinearVelScale:Vector="0.4,0.3,0.3"
    SimSpaceSettings:Sim Space Settings Structure=(WorldAlpha=1,VelocityScaleZ=1,DampingAlpha=1,MaxLinearVelocity=500,MaxAngularVelocity=10,MaxLinearAcceleration=500,MaxAngularAcceleration=5,ExternalLinearDragV=(X=0.2,Y=0.2,Z=0),ExternalLinearVelocity=(X=0,Y=0,Z=0),ExternalAngularVelocity=(X=0,Y=0,Z=0))
   ) -> (Pose) {
    bClampLinearTranslationLimitToRefPose=true
    bOverrideWorldGravity=true
    overridePhysicsAsset=&F_MED_FreightCalf_Physics
    simulationSpace=BaseBoneSpace
    alphaBoolBlend {blendOption=Linear}
    baseBoneRef {boneName=spine_03}
    componentAppliedLinearAccClamp=<5000 5000 5000>
    componentLinearAccScale=<0.3 0.2 0.2>
    componentLinearVelScale=<0.4 0.3 0.3>
    simSpaceSettings {
     maxAngularAcceleration=5
     maxAngularVelocity=10
     maxLinearAcceleration=500
     maxLinearVelocity=500
     worldAlpha=1
     externalLinearDragV=<0.2 0.2>
    }
   }
   n1 = Copy Pose From Mesh() {bUseAttachedParent=true}
   n2 = Control Rig(Source=Component To Local(ComponentPose=Local To Component(LocalPose=@n1))) {
    alphaBoolBlend {blendOption=Linear}
    controlRigAssetReference {blueprintRigClass=&/F_MED_FreightCalf_CtrlRig.F_MED_FreightCalf_CtrlRig_C}
   }
   n3 = "Constraint - dyn_chain_01_MAIN (P)"!(ComponentPose=Local To Component(LocalPose=@n2), Alpha=1, ConstraintWeights_0=1, ConstraintWeights_1:float=0) -> (Pose) {
    alphaBoolBlend {blendOption=Linear}
    boneToModify {boneName=dyn_chain_01_MAIN}
    constraintSetup [{targetBone {boneName=thigh_twist_01_r}}]
    constraintWeights [1,0]
   }
   n4 = "Constraint - dyn_chain_02_BASE (P)"!(ComponentPose=@n3, Alpha=1, ConstraintWeights_0=1, ConstraintWeights_1:float=0) -> (Pose) {
    alphaBoolBlend {blendOption=Linear}
    boneToModify {boneName=dyn_chain_02_BASE}
    constraintSetup [{targetBone {boneName=thigh_twist_01_r}}]
    constraintWeights [1,0]
   }
   n5 = "Constraint - dyn_chain_05_BASE (PPP)"!(ComponentPose=@n4, Alpha=1, ConstraintWeights_0=0.1513, ConstraintWeights_1:float=0.4123, ConstraintWeights_2:float=0.4364) -> (Pose) {
    alphaBoolBlend {blendOption=Linear}
    boneToModify {boneName=dyn_chain_05_BASE}
    constraintSetup [{targetBone {boneName=pelvis}}, {targetBone {boneName=thigh_twist_01_l}}, {targetBone {boneName=thigh_twist_01_r}}]
    constraintWeights [1,1,1]
   }
   n6 = "Constraint - dyn_chain_03_BASE (PPP)"!(ComponentPose=@n5, Alpha=1, ConstraintWeights_0=0.0597, ConstraintWeights_1:float=0.939, ConstraintWeights_2:float=0.0013) -> (Pose) {
    alphaBoolBlend {blendOption=Linear}
    boneToModify {boneName=dyn_chain_03_BASE}
    constraintSetup [{targetBone {boneName=pelvis}}, {targetBone {boneName=thigh_twist_01_l}}, {targetBone {boneName=thigh_twist_01_r}}]
    constraintWeights [1,1,1]
   }
   n7 = "Constraint - dyn_chain_04_BASE (P)"!(ComponentPose=@n6, Alpha=1, ConstraintWeights_0=1, ConstraintWeights_1:float=0) -> (Pose) {
    alphaBoolBlend {blendOption=Linear}
    boneToModify {boneName=dyn_chain_04_BASE}
    constraintSetup [{targetBone {boneName=pelvis}}]
    constraintWeights [1,0]
   }
   n8 = "Transform (Modify) Bone - Bone: dyn_chain_02_BASE"(ComponentPose=@n7, Alpha=1, Translation="0,3,0", Rotation="0,0,0", Scale="1,1,1") {translationMode=BMM_Additive, alphaBoolBlend {blendOption=Linear}, boneToModify {boneName=dyn_chain_02_BASE}, translation {y=3}}
   n9 = "Transform (Modify) Bone - Bone: dyn_chain_02_BASE"(ComponentPose=@n8, Alpha=1, Translation="0,-1,0", Rotation="0,0,0", Scale="1,1,1") {translationMode=BMM_Additive, alphaBoolBlend {blendOption=Linear}, boneToModify {boneName=dyn_chain_02_BASE}, translation {y=-1}}
   n10 = "Transform (Modify) Bone - Bone: dyn_chain_03_BASE"(ComponentPose=@n9, Alpha=1, Translation="-0.000,5.000,-2.560", Rotation="0,0,0", Scale="1,1,1") {translationMode=BMM_Additive, alphaBoolBlend {blendOption=Linear}, boneToModify {boneName=dyn_chain_03_BASE}, translation {y=5, z=-2.56}}
   n11 = "Transform (Modify) Bone - Bone: dyn_chain_03_BASE"(ComponentPose=@n10, Alpha=1, Translation="0,2,0", Rotation="0,0,0", Scale="1,1,1") {translationMode=BMM_Additive, alphaBoolBlend {blendOption=Linear}, boneToModify {boneName=dyn_chain_03_BASE}, translation {y=2}}
   n12 = "Transform (Modify) Bone - Bone: dyn_chain_05_BASE"(ComponentPose=@n11, Alpha=1, Translation="0,2,0", Rotation="0,0,0", Scale="1,1,1") {translationMode=BMM_Additive, alphaBoolBlend {blendOption=Linear}, boneToModify {boneName=dyn_chain_05_BASE}, translation {y=2}}
   n13 = "Transform (Modify) Bone - Bone: dyn_chain_04_BASE"(ComponentPose=@n12, Alpha=1, Translation="0,0,0", Rotation="7.263,24.181,-101.316", Scale="1,1,1") {
    rotationMode=BMM_Additive
    translationMode=BMM_Additive
    alphaBoolBlend {blendOption=Linear}
    boneToModify {boneName=dyn_chain_04_BASE}
    rotation=<rot 7.263 24.181 -101.316>
   }
   n14 = Anim Dynamics - Bone: dyn_clip_01(ComponentPose=@n13, Alpha=1, ExternalForce="0,0,0") {
    angularDampingOverride=1
    angularSpringConstant=20000
    bAngularSpring=true
    bOverrideAngularDamping=true
    bUseGravityOverride=true
    gravityScale=0.35
    simulationSpace=RootRelative
    alphaBoolBlend {blendOption=Linear}
    boundBone {boneName=dyn_clip_01}
    componentLinearAccScale=<1 1 1>
    componentLinearVelScale=<1 1 1>
    gravityOverride {z=-980}
    physicsBodyDefinitions [{
      boundBone {boneName=dyn_clip_01}
      boxExtents=<20 30 30>
      constraintSetup {angularLimitsMax=<0 60 0>, angularLimitsMin=<-30 -60 0>, angularTarget=<1 0 0>}
      localJointOffset=<4 0 0>
     }]
   }
   n15 = Anim Dynamics - Bone: dyn_clip_02(ComponentPose=@n14, Alpha=1, ExternalForce="0,0,0") {
    angularDampingOverride=1
    angularSpringConstant=20000
    bAngularSpring=true
    bOverrideAngularDamping=true
    bUseGravityOverride=true
    gravityScale=0.35
    simulationSpace=RootRelative
    alphaBoolBlend {blendOption=Linear}
    boundBone {boneName=dyn_clip_02}
    componentLinearAccScale=<1 1 1>
    componentLinearVelScale=<1 1 1>
    gravityOverride {z=-980}
    physicsBodyDefinitions [{
      boundBone {boneName=dyn_clip_02}
      boxExtents=<20 30 30>
      constraintSetup {angularLimitsMax=<0 60 0>, angularLimitsMin=<-30 -60 0>, angularTarget=<1 0 0>}
      localJointOffset=<4 0 0>
     }]
   }
   n16 = "Constraint - dyn_button (PPPPPPPP)"!(
    ComponentPose=@n15
    Alpha=1
    ConstraintWeights_0=0.0002
    ConstraintWeights_1:float=0.0025
    ConstraintWeights_2:float=0.7414
    ConstraintWeights_3:float=0.104
    ConstraintWeights_4:float=0.0245
    ConstraintWeights_5:float=0.1094
    ConstraintWeights_6:float=0.0139
    ConstraintWeights_7:float=0.0041
   ) -> (Pose) {
    alphaBoolBlend {blendOption=Linear}
    boneToModify {boneName=dyn_button}
    constraintSetup [
     {targetBone {boneName=upperarm_twist_01_l}}
     {targetBone {boneName=spine_02}}
     {targetBone {boneName=spine_04}}
     {targetBone {boneName=spine_03}}
     {targetBone {boneName=spine_05}}
     {targetBone {boneName=clavicle_l}}
     {targetBone {boneName=spine_01}}
     {targetBone {boneName=dyn_coller_01_l}}
    ]
    constraintWeights [1,1,1,1,1,1,1,1]
   }
   n17 = RigidBody!(ComponentPose=@n16, Alpha, OverrideWorldGravity="0,0,-980") -> (Pose) {
    bClampLinearTranslationLimitToRefPose=true
    bOverrideWorldGravity=true
    overridePhysicsAsset=&F_MED_FreightCalf_Physics
    simulationSpace=BaseBoneSpace
    alphaBoolBlend {blendOption=Linear}
    baseBoneRef {boneName=spine_03}
    componentLinearAccScale=<0.3 0.2 0.2>
    componentLinearVelScale=<0.4 0.3 0.3>
    simSpaceSettings {
     maxAngularAcceleration=5
     maxAngularVelocity=10
     maxLinearAcceleration=500
     maxLinearVelocity=500
     worldAlpha=1
     externalLinearDragV=<0.2 0.2>
    }
   }
   n18 = "Transform (Modify) Bone - Bone: dyn_jacket_03_c"!(ComponentPose=@n17, bAlphaBoolEnabled:bool=True, Translation="5,0,0", Rotation="0,0,0", Scale="1,1,1") -> (Pose) {
    alphaInputType=Bool
    rotationSpace=BCS_BoneSpace
    scaleMode=BMM_Additive
    scaleSpace=BCS_BoneSpace
    translationMode=BMM_Additive
    translationSpace=BCS_BoneSpace
    alphaBoolBlend {blendInTime=0.2, blendOption=Linear, blendOutTime=0.2}
    boneToModify {boneName=dyn_jacket_03_c}
    translation {x=5}
   }
   n19 = "Constraint - dyn_coller_01_l (P)"(ComponentPose=@n18, Alpha=0.4, ConstraintWeights_0=1) {
    alphaBoolBlend {blendOption=Linear}
    boneToModify {boneName=dyn_coller_01_l}
    constraintSetup [{targetBone {boneName=clavicle_l}}]
    constraintWeights [1]
   }
   n20 = "Constraint - dyn_coller_01_r (P)"(ComponentPose=@n19, Alpha=0.2, ConstraintWeights_0=1) {
    alphaBoolBlend {blendOption=Linear}
    boneToModify {boneName=dyn_coller_01_r}
    constraintSetup [{targetBone {boneName=clavicle_r}}]
    constraintWeights [1]
   }
   Output Pose(Result=Component To Local(ComponentPose=@n20)) {layerGroup=None}
  ]}
 EventGraph {nodes [
   Event BlueprintUpdateAnimation()
   > n1 = Sequence!(execute) -> (then_0, then_1, then_2:Exec, then_3:Exec, then_4:Exec)
   n2 = Get bIsSkyDiving()
   > Set collarSkydive_l(collarSkydive_l:double=SelectFloat(A=0.45, B=0, bPickA=@n2)) -> (Output_Get:double)
   > Set collarSkydive(collarSkydive:double=SelectFloat(A=0.15, B=0, bPickA=@n2)) -> (Output_Get:double)
   n3 = Gravity Override State Function DEPRECATED(
    execute=[@n1.then_1, @n1.then_3]
    CharacterPartAnimInstance=Self-Reference()
    GameTimeInSeconds=GetGameTimeInSeconds()
    GravityOverrideParams=Get Gravity Override Params()
    RBAN Params=Get RBAN Params()
   )
   > Set Gravity Override(Gravity Override=@n3)
   > Set RBAN Settings Out(RBAN Settings Out:RBANSetting Struct Structure=@n3.RBAN SettingsOut) -> (Output_Get:RBANSetting Struct Structure)
  ]}
]
variables [
 BodyMesh {type=SkeletalMeshComponent}
 collarBack {type=number}
 collarClav {type=number}
 collarSkydive {type=number}
 collarSkydive_l {type=number}
 Gravity Override {type=Vector}
 Gravity Override Params {type=GravityOverrideParamsStruct, default {skydive_Surfing {gravityOverride=<500 -700 600>, windOffset=<0 -600 400>}}}
 headRotX {type=number}
 headRotZNeg {type=number}
 headTwistNeg {type=number}
 headTwistPos {type=number}
 NewVar {type=boolean}
 RBAN Params {
  type=RBANParamsStruct
  default {
   dBNO {
    componentLinearAccScale=<0.3 0.2 0.2>
    componentLinearVelScale=<0.4 0.3 0.3>
    simSpaceSettings {
     maxAngularAcceleration=5
     maxAngularVelocity=10
     maxLinearAcceleration=500
     maxLinearVelocity=500
     velocityScaleZ=1
     externalLinearDragV=<0.2 0.2 0>
    }
   }
   dBNO_Carried {
    componentLinearAccScale=<0.3 0.2 0.2>
    componentLinearVelScale=<0.4 0.3 0.3>
    simSpaceSettings {
     maxAngularAcceleration=5
     maxAngularVelocity=10
     maxLinearAcceleration=500
     maxLinearVelocity=500
     velocityScaleZ=1
     externalLinearDragV=<0.2 0.2 0>
    }
   }
   emoteOrMelee {
    componentLinearAccScale=<0.3 0.2 0.2>
    componentLinearVelScale=<0.4 0.3 0.3>
    simSpaceSettings {
     maxAngularAcceleration=5
     maxAngularVelocity=10
     maxLinearAcceleration=500
     maxLinearVelocity=500
     velocityScaleZ=1
     externalLinearDragV=<0.2 0.2 0>
    }
   }
   front_End {
    componentLinearAccScale=<0.3 0.2 0.2>
    componentLinearVelScale=<0.4 0.3 0.3>
    simSpaceSettings {
     maxAngularAcceleration=3
     maxAngularVelocity=5
     maxLinearAcceleration=500
     maxLinearVelocity=500
     velocityScaleZ=1
     externalLinearDragV=<0.2 0.2 0>
    }
   }
   onGround_Crouch_Moving {
    componentLinearAccScale=<0.3 0.2 0.2>
    componentLinearVelScale=<0.4 0.3 0.3>
    simSpaceSettings {
     maxAngularAcceleration=5
     maxAngularVelocity=10
     maxLinearAcceleration=500
     maxLinearVelocity=500
     velocityScaleZ=1
     externalLinearDragV=<0.2 0.2 0>
    }
   }
   onGround_Falling {
    componentLinearAccScale=<0.3 0.2 0.2>
    componentLinearVelScale=<0.4 0.3 0.3>
    simSpaceSettings {
     maxAngularAcceleration=5
     maxAngularVelocity=10
     maxLinearAcceleration=500
     maxLinearVelocity=500
     velocityScaleZ=1
     externalLinearDragV=<0.2 0.2 0>
    }
   }
   onGround_Moving {
    componentLinearAccScale=<0.3 0.2 0.2>
    componentLinearVelScale=<0.4 0.3 0.3>
    simSpaceSettings {
     maxAngularAcceleration=5
     maxAngularVelocity=10
     maxLinearAcceleration=500
     maxLinearVelocity=500
     velocityScaleZ=1
     externalLinearDragV=<0.2 0.2 0>
    }
   }
   onGround_Standing {
    componentLinearAccScale=<0.3 0.2 0.2>
    componentLinearVelScale=<0.4 0.3 0.3>
    simSpaceSettings {
     maxAngularAcceleration=5
     maxAngularVelocity=10
     maxLinearAcceleration=500
     maxLinearVelocity=500
     velocityScaleZ=1
     externalLinearDragV=<0.2 0.2 0>
    }
   }
   rail_Grinding {
    componentLinearAccScale=<0.3 0.2 0.2>
    componentLinearVelScale=<0.4 0.3 0.3>
    simSpaceSettings {
     maxAngularAcceleration=5
     maxAngularVelocity=10
     maxLinearAcceleration=500
     maxLinearVelocity=500
     velocityScaleZ=1
     worldAlpha=1
     externalLinearDragV=<0.2 0.2 0>
    }
   }
   riding_Moving {
    componentLinearAccScale=<0.3 0.2 0.2>
    componentLinearVelScale=<0.4 0.3 0.3>
    simSpaceSettings {
     maxAngularAcceleration=5
     maxAngularVelocity=10
     maxLinearAcceleration=500
     maxLinearVelocity=500
     velocityScaleZ=1
     externalLinearDragV=<0.2 0.2 0>
    }
   }
   riding_Moving_Targeting {
    componentLinearAccScale=<0.3 0.2 0.2>
    componentLinearVelScale=<0.4 0.3 0.3>
    simSpaceSettings {
     maxAngularAcceleration=5
     maxAngularVelocity=10
     maxLinearAcceleration=500
     maxLinearVelocity=500
     velocityScaleZ=1
     externalLinearDragV=<0.2 0.2 0>
    }
   }
   riding_Standing {
    componentLinearAccScale=<0.3 0.2 0.2>
    componentLinearVelScale=<0.4 0.3 0.3>
    simSpaceSettings {
     maxAngularAcceleration=5
     maxAngularVelocity=10
     maxLinearAcceleration=500
     maxLinearVelocity=500
     velocityScaleZ=1
     externalLinearDragV=<0.2 0.2 0>
    }
   }
   skydive_DivingDown {
    componentLinearAccScale=<0.3 0.2 0.2>
    componentLinearVelScale=<0.4 0.3 0.3>
    simSpaceSettings {
     maxAngularAcceleration=5
     maxAngularVelocity=10
     maxLinearAcceleration=500
     maxLinearVelocity=500
     velocityScaleZ=1
     externalLinearDragV=<0.2 0.2 0>
    }
   }
   skydive_FloatingDown {
    componentLinearAccScale=<0.3 0.2 0.2>
    componentLinearVelScale=<0.4 0.3 0.3>
    simSpaceSettings {
     maxAngularAcceleration=5
     maxAngularVelocity=10
     maxLinearAcceleration=500
     maxLinearVelocity=500
     velocityScaleZ=1
     externalLinearDragV=<0.2 0.2 0>
    }
   }
   skydive_GoingUp {
    componentLinearAccScale=<0.3 0.2 0.2>
    componentLinearVelScale=<0.4 0.3 0.3>
    simSpaceSettings {
     maxAngularAcceleration=5
     maxAngularVelocity=10
     maxLinearAcceleration=500
     maxLinearVelocity=500
     velocityScaleZ=1
     externalLinearDragV=<0.2 0.2 0>
    }
   }
   skydive_Parachuting {
    componentLinearAccScale=<0.3 0.2 0.2>
    componentLinearVelScale=<0.4 0.3 0.3>
    simSpaceSettings {
     maxAngularAcceleration=5
     maxAngularVelocity=10
     maxLinearAcceleration=500
     maxLinearVelocity=500
     velocityScaleZ=1
     externalLinearDragV=<0.2 0.2 0>
    }
   }
   skydive_Surfing {
    componentLinearAccScale=<0.3 0.2 0.2>
    componentLinearVelScale=<0.4 0.3 0.3>
    simSpaceSettings {
     maxAngularAcceleration=5
     maxAngularVelocity=10
     maxLinearAcceleration=500
     maxLinearVelocity=500
     velocityScaleZ=1
     externalLinearDragV=<0.2 0.2 0>
    }
   }
   swimming {
    componentLinearAccScale=<0.3 0.2 0.2>
    componentLinearVelScale=<0.4 0.3 0.3>
    simSpaceSettings {
     maxAngularAcceleration=5
     maxAngularVelocity=10
     maxLinearAcceleration=500
     maxLinearVelocity=500
     velocityScaleZ=1
     externalLinearDragV=<0.2 0.2 0>
    }
   }
   tactical_Sprint {
    componentLinearAccScale=<0.3 0.2 0.2>
    componentLinearVelScale=<0.4 0.3 0.3>
    simSpaceSettings {
     maxAngularAcceleration=5
     maxAngularVelocity=10
     maxLinearAcceleration=500
     maxLinearVelocity=500
     velocityScaleZ=1
     externalLinearDragV=<0.2 0.2 0>
    }
   }
   ziplining {
    componentLinearAccScale=<0.3 0.2 0.2>
    componentLinearVelScale=<0.4 0.3 0.3>
    simSpaceSettings {
     maxAngularAcceleration=5
     maxAngularVelocity=10
     maxLinearAcceleration=500
     maxLinearVelocity=500
     velocityScaleZ=1
     externalLinearDragV=<0.2 0.2 0>
    }
   }
  }
 }
 RBAN Settings Out {
  type=RBANSettingStruct
  default {
   componentLinearAccScale=<0 0 0>
   componentLinearVelScale=<0 0 0>
   simSpaceSettings {
    dampingAlpha=1
    maxAngularAcceleration=1000
    maxAngularVelocity=3
    maxLinearAcceleration=300
    maxLinearVelocity=700
    velocityScaleZ=0
    worldAlpha=1
    externalAngularVelocity=<0 0 0>
    externalLinearDragV=<0 0.5 0>
    externalLinearVelocity=<0 0 0>
   }
  }
 }
]
