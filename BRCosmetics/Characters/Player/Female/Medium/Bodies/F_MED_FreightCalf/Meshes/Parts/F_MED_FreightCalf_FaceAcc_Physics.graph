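# Physics asset for the FreightCalf female face accessory mesh, defining bodies and joints for its dynamic hair.
# A six-link capsule hair chain hangs off a base locked to the head, with a strand branching from the fifth link and a front piece from the third.
# Hair joints have limited swing and mostly locked twist, with the branch strand links restricted to one swing axis.
# Head, neck, upper spine and two collar boxes are joined by free constraints.
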PhysicsAsset /BRCosmetics/Characters/Player/Female/Medium/Bodies/F_MED_FreightCalf/Meshes/Parts/F_MED_FreightCalf_FaceAcc_Physics
bodies {
 dyn_coller_01_l [{
   shapeType=Box
   extentX=2.352
   extentY=8.3217
   extentZ=18.4879
   shapeName=dyn_coller_01_l_box
   center=<4.2151 0.0105 -2.1235>
   rotation=<rot -1.2014 -87.9278 176.7272>
  }]
 dyn_coller_01_r [{
   shapeType=Box
   extentX=2.352
   extentY=8.3217
   extentZ=18.4879
   shapeName=dyn_coller_01_r_box
   center=<-4.2151 -0.0105 2.1235>
   rotation=<rot -1.2014 -87.9278 -3.2728>
  }]
 dyn_hairStrandFront_BASE [{
   shapeType=Capsule
   length=3.51
   radius=0.7748
   shapeName=dyn_hairStrandFront_BASE_capsule
   center=<3 0 0.1268>
   rotation=<rot -85 0.0002 15.0001>
  }]
 dyn_hairStrand_01 [{
   shapeType=Capsule
   length=1.9833
   radius=1.0242
   shapeName=dyn_hairStrand_01_capsule
   center=<3 0.1709 0.0263>
   rotation=<rot -90 89.7893 -89.7891>
  }]
 dyn_hairStrand_02 [{
   shapeType=Capsule
   length=1.566
   radius=0.7694
   shapeName=dyn_hairStrand_02_capsule
   center=<3 0.011 0.2475>
   rotation=<rot 0 96.6261 -97.8649>
  }]
 dyn_hairStrand_03 [{
   shapeType=Capsule
   length=0.505
   radius=0.505
   shapeName=dyn_hairStrand_03_capsule
   center=<1.3489 0.4107 -0.3748>
   rotation=<rot 0 -78.8892 -76.742>
  }]
 dyn_hairStrand_BASE [{shapeType=Sphere, radius=0.505, shapeName=dyn_hairStrand_BASE_sphere}]
 dyn_hair_01 [{shapeType=Capsule, length=2.0387, radius=1.5, shapeName=dyn_hair_01_capsule, center=<3 0 0>, rotation=<rot -0.0001 0.0002 -70>}]
 dyn_hair_02 [{shapeType=Capsule, length=8.0827, radius=1.0403, shapeName=dyn_hair_02_capsule, center=<3 3.4653 -0.5>, rotation=<rot -0.0002 0.0004 -85>}]
 dyn_hair_03 [{shapeType=Capsule, length=8.182, radius=1.1665, shapeName=dyn_hair_03_capsule, center=<3 3.4653 -0.5>, rotation=<rot -0.0001 0.0003 -85>}]
 dyn_hair_04 [{shapeType=Capsule, length=2.8005, radius=0.9023, shapeName=dyn_hair_04_capsule, center=<3 0 0>, rotation=<rot -0.0002 0.0005 -85>}]
 dyn_hair_05 [{shapeType=Capsule, length=1.7062, radius=1.0544, shapeName=dyn_hair_05_capsule, center=<3 0 0>, rotation=<rot -0.0001 0.0002 -70>}]
 dyn_hair_06 [{shapeType=Capsule, length=3.5515, radius=0.5, shapeName=dyn_hair_06_capsule, center=<3 0 0>, rotation=<rot 15.1888 -48.2362 -76.832>}]
 dyn_hair_BASE [{shapeType=Sphere, radius=0.505, shapeName=dyn_hair_BASE_sphere}]
 head [{
   shapeType=Capsule
   length=7.8847
   radius=5.8557
   shapeName=head_capsule
   center=<2.2183 3.519 -0.2776>
   rotation=<rot 0.9994 82.6394 -88.5723>
  }]
 neck_01 [{shapeType=Capsule, length=10, radius=5, shapeName=neck_01_capsule, center=<3.918 0 0>, rotation=<rot 0 -90.0001 90>}]
 spine_05 [{shapeType=Capsule, length=12.1303, radius=10.6891, shapeName=spine_05_capsule, center=<0 -1.5769 0>, rotation=<rot 0 -90 0>}]
}
constraints [
 {bone1Name=dyn_hair_01, bone2Name=dyn_hair_BASE, swing1LimitDegrees=10, swing2LimitDegrees=5, twistLimitDegrees=45, twistMotion=Locked}
 {bone1Name=dyn_hair_02, bone2Name=dyn_hair_01, swing1LimitDegrees=20, swing2LimitDegrees=20, twistLimitDegrees=45, twistMotion=Locked}
 {bone1Name=dyn_hair_03, bone2Name=dyn_hair_02, swing1LimitDegrees=20, swing2LimitDegrees=20, twistLimitDegrees=45, twistMotion=Locked}
 {bone1Name=dyn_hair_04, bone2Name=dyn_hair_03, swing1LimitDegrees=20, swing2LimitDegrees=20, twistLimitDegrees=45, twistMotion=Locked}
 {bone1Name=dyn_hair_05, bone2Name=dyn_hair_04, swing1LimitDegrees=40, swing2LimitDegrees=20, twistLimitDegrees=45, twistMotion=Locked}
 {bone1Name=dyn_hair_06, bone2Name=dyn_hair_05, swing1LimitDegrees=20, swing2LimitDegrees=20, twistLimitDegrees=45, twistMotion=Locked}
 {
  bone1Name=dyn_hair_BASE
  bone2Name=head
  swing1LimitDegrees=45
  swing1Motion=Locked
  swing2LimitDegrees=45
  swing2Motion=Locked
  twistLimitDegrees=45
  twistMotion=Locked
 }
 {
  bone1Name=head
  bone2Name=neck_01
  swing1LimitDegrees=45
  swing1Motion=Free
  swing2LimitDegrees=45
  swing2Motion=Free
  twistLimitDegrees=45
  twistMotion=Free
 }
 {bone1Name=dyn_hairStrand_BASE, bone2Name=dyn_hair_05, swing1LimitDegrees=45, swing2LimitDegrees=45, twistLimitDegrees=45, twistMotion=Locked}
 {
  bone1Name=dyn_hairStrand_01
  bone2Name=dyn_hairStrand_BASE
  swing1LimitDegrees=20
  swing2LimitDegrees=20
  swing2Motion=Locked
  twistLimitDegrees=45
  twistMotion=Locked
 }
 {
  bone1Name=dyn_hairStrand_02
  bone2Name=dyn_hairStrand_01
  swing1LimitDegrees=20
  swing2LimitDegrees=20
  swing2Motion=Locked
  twistLimitDegrees=45
  twistMotion=Locked
 }
 {
  bone1Name=dyn_hairStrand_03
  bone2Name=dyn_hairStrand_02
  swing1LimitDegrees=20
  swing2LimitDegrees=20
  swing2Motion=Locked
  twistLimitDegrees=45
  twistMotion=Locked
 }
 {bone1Name=dyn_hairStrandFront_BASE, bone2Name=dyn_hair_03, swing1LimitDegrees=10, swing2LimitDegrees=10, twistLimitDegrees=2}
 {
  bone1Name=neck_01
  bone2Name=spine_05
  swing1LimitDegrees=45
  swing1Motion=Free
  swing2LimitDegrees=45
  swing2Motion=Free
  twistLimitDegrees=45
  twistMotion=Free
 }
 {
  bone1Name=dyn_coller_01_r
  bone2Name=spine_05
  swing1LimitDegrees=45
  swing1Motion=Free
  swing2LimitDegrees=45
  swing2Motion=Free
  twistLimitDegrees=45
  twistMotion=Free
 }
 {
  bone1Name=dyn_coller_01_l
  bone2Name=spine_05
  swing1LimitDegrees=45
  swing1Motion=Free
  swing2LimitDegrees=45
  swing2Motion=Free
  twistLimitDegrees=45
  twistMotion=Free
 }
]
