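# Post-copy AnimBP for the FreightCalf female face accessory, adding secondary motion to hair and attaching a lip ring.
# It copies the parent mesh pose, layering collar bones from Body_Mesh, which is cached when character customization loads.
# An additive pitch tilts the hair base before spring-driven AnimDynamics moves the single-hair chain and back strand.
# A weighted constraint binds the lip ring to head and facial bones, then RigidBody simulates the physics asset relative to spine_05.
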AnimBlueprint /BRCosmetics/Characters/Player/Female/Medium/Bodies/F_MED_FreightCalf/Meshes/Parts/F_MED_FreightCalf_FaceAcc_AnimBP
target_skeleton=&/F_MED_FreightCalf_FaceAcc_Skeleton
graphs [
 AnimGraph {nodes [
   n1 = Layered blend per bone(
    BasePose=Copy Pose From Mesh() {bCopyCurves=true, bUseAttachedParent=true}
    BlendPoses_0=Copy Pose From Mesh(SourceMeshComponent=Get Body_Mesh())
   ) {layerSetup [{branchFilters [{boneName=dyn_coller_01_l}, {boneName=dyn_coller_01_r}, {boneName=dyn_coller_01_c}]}]}
   n2 = "Transform (Modify) Bone - Bone: dyn_hairSingle_BASE"(ComponentPose=Local To Component(LocalPose=@n1), Alpha=1, Translation="0,0,0", Rotation="-15.000,0.000,-0.000", Scale="1,1,1") {rotationMode=BMM_Additive, alphaBoolBlend {blendOption=Linear}, boneToModify {boneName=dyn_hairSingle_BASE}, rotation {pitch=-15}}
   n3 = Anim Dynamics - Bone: dyn_hairSingle_04(ComponentPose=@n2, Alpha=1, ExternalForce="0,0,0") {
    angularDampingOverride=1
    angularSpringConstant=20000
    bAngularSpring=true
    bOverrideAngularDamping=true
    bUseGravityOverride=true
    gravityScale=0.35
    simulationSpace=RootRelative
    alphaBoolBlend {blendOption=Linear}
    boundBone {boneName=dyn_hairSingle_04}
    componentLinearAccScale=<1 1 1>
    componentLinearVelScale=<1 1 1>
    gravityOverride {z=-980}
    physicsBodyDefinitions [{
      boundBone {boneName=dyn_hairSingle_04}
      boxExtents=<20 30 30>
      constraintSetup {angularLimitsMax=<15 0 30>, angularLimitsMin=<-15 0 0>, angularTarget=<1 0 0>}
      localJointOffset=<4 0 0>
     }]
   }
   n4 = Anim Dynamics - Bone: dyn_hairSingle_03(ComponentPose=@n3, Alpha=1, ExternalForce="0,0,0") {
    angularDampingOverride=1
    angularSpringConstant=20000
    bAngularSpring=true
    bOverrideAngularDamping=true
    bUseGravityOverride=true
    gravityScale=0.35
    simulationSpace=RootRelative
    alphaBoolBlend {blendOption=Linear}
    boundBone {boneName=dyn_hairSingle_03}
    componentLinearAccScale=<1 1 1>
    componentLinearVelScale=<1 1 1>
    gravityOverride {z=-980}
    physicsBodyDefinitions [{
      boundBone {boneName=dyn_hairSingle_03}
      boxExtents=<20 30 30>
      constraintSetup {angularLimitsMax=<15 0 30>, angularLimitsMin=<-15 0 0>, angularTarget=<1 0 0>}
      localJointOffset=<4 0 0>
     }]
   }
   n5 = Anim Dynamics - Bone: dyn_hairSingle_02(ComponentPose=@n4, Alpha=1, ExternalForce="0,0,0") {
    angularDampingOverride=1
    angularSpringConstant=20000
    bAngularSpring=true
    bOverrideAngularDamping=true
    bUseGravityOverride=true
    gravityScale=0.35
    simulationSpace=RootRelative
    alphaBoolBlend {blendOption=Linear}
    boundBone {boneName=dyn_hairSingle_02}
    componentLinearAccScale=<1 1 1>
    componentLinearVelScale=<1 1 1>
    gravityOverride {z=-980}
    physicsBodyDefinitions [{
      boundBone {boneName=dyn_hairSingle_02}
      boxExtents=<20 30 30>
      constraintSetup {angularLimitsMax=<15 0 30>, angularLimitsMin=<-15 0 0>, angularTarget=<1 0 0>}
      localJointOffset=<4 0 0>
     }]
   }
   n6 = Anim Dynamics - Bone: dyn_hairSingle_01(ComponentPose=@n5, Alpha=1, ExternalForce="0,0,0") {
    angularDampingOverride=1
    angularSpringConstant=20000
    bAngularSpring=true
    bOverrideAngularDamping=true
    bUseGravityOverride=true
    gravityScale=0.35
    simulationSpace=RootRelative
    alphaBoolBlend {blendOption=Linear}
    boundBone {boneName=dyn_hairSingle_01}
    componentLinearAccScale=<1 1 1>
    componentLinearVelScale=<1 1 1>
    gravityOverride {z=-980}
    physicsBodyDefinitions [{
      boundBone {boneName=dyn_hairSingle_01}
      boxExtents=<20 30 30>
      constraintSetup {angularLimitsMax=<15 0 30>, angularLimitsMin=<-15 0 0>, angularTarget=<1 0 0>}
      localJointOffset=<4 0 0>
     }]
   }
   n7 = Anim Dynamics - Bone: dyn_backStrand(ComponentPose=@n6, Alpha=1, ExternalForce="0,0,0") {
    angularDampingOverride=1
    angularSpringConstant=20000
    bAngularSpring=true
    bOverrideAngularDamping=true
    bUseGravityOverride=true
    gravityScale=0.35
    simulationSpace=RootRelative
    alphaBoolBlend {blendOption=Linear}
    boundBone {boneName=dyn_backStrand}
    componentLinearAccScale=<1 1 1>
    componentLinearVelScale=<1 1 1>
    gravityOverride {z=-980}
    physicsBodyDefinitions [{
      boundBone {boneName=dyn_backStrand}
      boxExtents=<20 20 20>
      constraintSetup {angularLimitsMax=<15 15 15>, angularLimitsMin=<-15 -15 -15>, angularTarget=<1 0 0>}
      localJointOffset=<4 0 0>
     }]
   }
   n8 = "Constraint - dyn_lipRing (PPPPPPPP)"!(
    ComponentPose=@n7
    Alpha=1
    ConstraintWeights_0=0.006
    ConstraintWeights_1:float=0.0013
    ConstraintWeights_2:float=0.1212
    ConstraintWeights_3:float=0.1874
    ConstraintWeights_4:float=0.4581
    ConstraintWeights_5:float=0.1583
    ConstraintWeights_6:float=0.0662
    ConstraintWeights_7:float=0.0015
   ) -> (Pose) {
    alphaBoolBlend {blendOption=Linear}
    boneToModify {boneName=dyn_lipRing}
    constraintSetup [
     {targetBone {boneName=head}}
     {targetBone {boneName=FACIAL_L_NasolabialBulge}}
     {targetBone {boneName=FACIAL_L_LipCorner}}
     {targetBone {boneName=FACIAL_C_LipLower}}
     {targetBone {boneName=FACIAL_L_LipLower}}
     {targetBone {boneName=FACIAL_L_LipLowerOuter}}
     {targetBone {boneName=FACIAL_C_Jaw}}
     {targetBone {boneName=FACIAL_C_Chin}}
    ]
    constraintWeights [1,1,1,1,1,1,1,1]
   }
   n9 = RigidBody!(ComponentPose=@n8, Alpha, OverrideWorldGravity:Vector="0,0,-980") -> (Pose) {
    bClampLinearTranslationLimitToRefPose=true
    bOverrideWorldGravity=true
    overridePhysicsAsset=&/F_MED_FreightCalf_FaceAcc_Physics
    simulationSpace=BaseBoneSpace
    alphaBoolBlend {blendOption=Linear}
    baseBoneRef {boneName=spine_05}
    simSpaceSettings {
     maxAngularAcceleration=10
     maxAngularVelocity=10
     maxLinearAcceleration=500
     maxLinearVelocity=500
     worldAlpha=1
     externalLinearDragV=<0.2 0.2>
    }
   }
   Output Pose(Result=Component To Local(ComponentPose=@n9)) {layerGroup=None}
  ]}
 EventGraph {nodes [
   Event OnCharacterCustomizationLoaded()
   > n1 = Cast To PlayerPawn_Athena(Object=TryGetPawnOwner())
   > Set Body_Mesh(Body_Mesh=GetSkeletalMeshForPartType(self=@n1))
   Event BlueprintUpdateAnimation()
  ]}
]
variables [Body_Mesh {type=SkeletalMeshComponent}]
